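; Dojo table prop Blueprint (A2) that randomizes the yaw of each instance in its HierarchicalInstancedStaticMesh component.
; The construction script seeds RandomStream from the actor's location, making the variation deterministic per placement.
; It then replaces every instance's local yaw with a random value within plus or minus RotationAmount, keeping location, pitch, roll and scale.
; On destruction it destroys its Box component; BeginPlay, Tick and overlap events are disabled stubs.
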(event EventBeginPlay
  (|CallParentFunction) ; This node is disabled and will not be called.\nDrag off pins to build functionality.
)

(event Collision|EventActorBeginOverlap
  (params (OtherActor :type Actor%20Object%20Reference))
  (|CallParentFunction OtherActor) ; This node is disabled and will not be called.\nDrag off pins to build functionality.
)

(event EventTick
  (params (DeltaSeconds :type Float%20%28single-precision%29))
  (|CallParentFunction DeltaSeconds) ; This node is disabled and will not be called.\nDrag off pins to build functionality.
)

(event EventDestroyed
  (Components|DestroyComponent (Variables|Default|GetBox)))

(fn ConstructionScript
  (bind _hierarchicalinstancedstaticmesh (Variables|Default|GetHierarchicalInstancedStaticMesh))
  (|CallParentFunction)
  (Variables|Default|SetRandomStream (Math|Random|MakeRandomStreamfromLocation (Transformation|GetActorLocation) 1.0 true))
  (for _index (range (Components|InstancedStaticMesh|GetInstanceCount _hierarchicalinstancedstaticmesh))
    (Components|InstancedStaticMesh|UpdateInstanceTransform (Variables|Default|GetHierarchicalInstancedStaticMesh) _index (Math|Transform|MakeTransform (.location (Components|InstancedStaticMesh|GetInstanceTransform _hierarchicalinstancedstaticmesh _index false)) (Math|Rotator|MakeRotator (.roll (.rotation (Components|InstancedStaticMesh|GetInstanceTransform _hierarchicalinstancedstaticmesh _index false))) (.pitch (.rotation (Components|InstancedStaticMesh|GetInstanceTransform _hierarchicalinstancedstaticmesh _index false))) (Math|Random|RandomFloatinRangefromStream (Variables|Default|GetRandomStream) (- (Variables|Default|GetRotationAmount) (* (Variables|Default|GetRotationAmount) 2.0)) (Variables|Default|GetRotationAmount))) (.scale (Components|InstancedStaticMesh|GetInstanceTransform _hierarchicalinstancedstaticmesh _index false))) false false false)))
